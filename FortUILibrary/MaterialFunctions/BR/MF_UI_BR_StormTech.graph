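# UI material function for Battle Royale "StormTech" that draws an animated, screen-space grid of rounded-square cells at two scales, main and sub.
# Tiled, rotatable screen UVs feed per-cell random values whose time-driven sine waves modulate each cell's visibility.
# Perspective-warped box SDFs form fills and outlines, with 2x2 symmetry and a derivative-based seam mask on sub cells.
# It outputs the combined Grid plus separate cell layers and masks.
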MaterialFunction /FortUILibrary/MaterialFunctions/BR/MF_UI_BR_StormTech
graph {
 properties {bExposeToLibrary=true, previewMaterialDomain=MD_UI}
 nodes [
  n1 = NamedRerouteUsage()
  n2 = NamedRerouteUsage()
  FunctionOutput(None=Saturate(None=Add(A=@n1, B=@n2))) {outputName=Grid}
  n3 = NamedRerouteUsage()
  n4 = MaterialFunctionCall!(UV=Frac(None=@n3), Swizzle UV) {materialFunction=/FortUILibrary/MaterialFunctions/Basics/MF_UI_Symmetry_2x2_New}
  n5 = Constant() {r=-0.0625}
  MaterialExpressionFunctionInput_13 = FunctionInput() {
   bUsePreviewValueAsDefault=true
   inputName=Sub Min/Max Perspective
   inputType=FunctionInput_Vector2
   sortPriority=212
   previewValue=<0.1 0 0 1>
  }
  n6 = Convert(None=@MaterialExpressionFunctionInput_13) {
   nodeName=Make Vector2
   convertInputs [{
     type=Vector2
     expressionInput {
      expression=&self:MaterialExpressionFunctionInput_13
      inputName=None
      mask=0
      maskA=0
      maskB=0
      maskG=0
      maskR=0
      outputIndex=0
     }
    }]
   convertOutputs [{type=Scalar}, {type=Scalar}]
  }
  n7 = NamedRerouteUsage()
  n8 = FunctionInput() {bUsePreviewValueAsDefault=true, inputName=Sub Min/Max Scale, inputType=FunctionInput_Vector2, sortPriority=210, previewValue=<0.75 0 0 1>}
  MaterialExpressionMax_0 = Max(A=@n8) {constB=0}
  n9 = Convert(None=@MaterialExpressionMax_0) {
   nodeName=Make Vector2
   convertInputs [{
     type=Vector2
     expressionInput {
      expression=&self:MaterialExpressionMax_0
      inputName=None
      mask=0
      maskA=0
      maskB=0
      maskG=0
      maskR=0
      outputIndex=0
     }
    }]
   convertOutputs [{type=Scalar}, {type=Scalar}]
  }
  n10 = MaterialFunctionCall!(
   UVs=MaterialFunctionCall!(
    UV=Add(A=@n4.Symmetry, B=@n5)
    Perspective Rotation=NamedRerouteUsage()
    Perspective Amount=LinearInterpolate(A=@n6, B=@n6, Alpha=@n7) {constA=0.1, constB=0}
   ) {materialFunction=/FortUILibrary/MaterialFunctions/Basics/MF_UI_Perspective}.Perspective UVs
   Size=LinearInterpolate(A=@n9, B=@n9, Alpha=NamedRerouteUsage()) {constA=0.75, constB=0}
   Corner Roundness=Constant() {r=0.2}
   Preserve Roundness=Constant()
   Pixel Size
  ) {materialFunction=/FortUILibrary/MaterialFunctions/Shapes/SDFs/MF_SDF_Box_V4}
  n11 = MaterialFunctionCall(DistanceField=@n10.SDF) {materialFunction=&MF_UI_AntiAliasing}
  FunctionOutput(None=Multiply(A=NamedRerouteUsage(), B=@n11.Inner)) {outputName=Sub Mask, sortPriority=11}
  MaterialFunctionCall() {materialFunction=&GetUserInterfaceUV}
  n12 = FunctionInput() {bUsePreviewValueAsDefault=true, inputName=Tiling, inputType=FunctionInput_Scalar, sortPriority=5, previewValue=<8 0 0 1>}
  n13 = Multiply(
   A=FunctionInput(Preview=ScreenPosition().ViewportUV) {bUsePreviewValueAsDefault=true, inputName=Screen UV, inputType=FunctionInput_Vector2}
   B=NamedRerouteDeclaration(None=@n12) {name=Tile Amount, nodeColor=<rgba 0 0 0 1>}
  ) {constB=8}
  MaterialFunctionCall(UV=@n13) {materialFunction=&MF_UI_UVScaling}
  MaterialExpressionFunctionInput_6 = FunctionInput() {bUsePreviewValueAsDefault=true, inputName=Sub Min/Max Opacity, inputType=FunctionInput_Vector2, sortPriority=205, previewValue=<0 0.5 0 1>}
  n14 = Convert(None=@MaterialExpressionFunctionInput_6) {
   nodeName=Make Vector2
   convertInputs [{
     type=Vector2
     expressionInput {
      expression=&self:MaterialExpressionFunctionInput_6
      inputName=None
      mask=0
      maskA=0
      maskB=0
      maskG=0
      maskR=0
      outputIndex=0
     }
    }]
   convertOutputs [{type=Scalar}, {type=Scalar}]
  }
  n15 = FunctionInput() {bUsePreviewValueAsDefault=true, inputName=Sub Speed, inputType=FunctionInput_Scalar, sortPriority=250, previewValue=<0.1 0 0 1>}
  n16 = Constant() {r=2}
  n17 = Floor(None=NamedRerouteDeclaration(None=Multiply(A=@n3, B=@n16) {constB=8}) {name=Sub UV, nodeColor=<rgba 0 0 0 1>})
  n18 = NamedRerouteUsage()
  n19 = Subtract(
   A=Multiply(A=NamedRerouteUsage(), B=@n15) {constB=0.1}
   B=NamedRerouteDeclaration(None=MaterialFunctionCall!(UV=@n17, Seed=@n18) {materialFunction=&MF_UI_Random}.Result) {name=Sub Cells Random, nodeColor=<rgba 0 0 0 1>}
  )
  n20 = LinearInterpolate(
   A=Multiply(A=FunctionInput() {bUsePreviewValueAsDefault=true, inputName=Sub Reduction, inputType=FunctionInput_Scalar, sortPriority=295}) {constB=-1}
   Alpha=ConstantBiasScale(None=Sine(None=@n19))
  ) {constA=-0.5}
  n21 = FunctionInput() {bUsePreviewValueAsDefault=true, inputName=Sub Outline Thickness, inputType=FunctionInput_Scalar, sortPriority=245, previewValue=<0.1 0 0 1>}
  n22 = MaterialFunctionCall!(DistanceField=@n10.SDF, Threshold, Thickness=@n21, Alignment=Constant() {r=-1}, Smoothing) {materialFunction=/FortUILibrary/MaterialFunctions/Shapes/Modifiers/MF_UI_ShapeOutline}
  n23 = NamedRerouteUsage()
  n24 = Frac(None=NamedRerouteUsage())
  MaterialExpressionMax_1 = Max(A=Abs(None=DDX(Value=@n24)), B=Abs(None=DDY(Value=@n24)))
  n25 = Convert(None=@MaterialExpressionMax_1) {
   nodeName=Make Vector2
   convertInputs [{
     type=Vector2
     expressionInput {
      expression=&self:MaterialExpressionMax_1
      inputName=None
      mask=0
      maskA=0
      maskB=0
      maskG=0
      maskR=0
      outputIndex=0
     }
    }]
   convertOutputs [{type=Scalar}, {type=Scalar}]
  }
  n26 = OneMinus(None=Round(None=Max(A=@n25, B=@n25)))
  n27 = Multiply(
   A=NamedRerouteDeclaration(None=Saturate(None=@n20)) {name=Animated Sub Cells, nodeColor=<rgba 0 0 0 1>}
   B=Multiply(
    A=LinearInterpolate(A=@n11.Inner, B=@n22.Outline, Alpha=@n23)
    B=NamedRerouteDeclaration(None=@n26) {name=Derivative Hack, nodeColor=<rgba 0 0 0 1>}
   )
  )
  NamedRerouteDeclaration(None=FunctionOutput(None=LinearInterpolate(A=@n14, B=@n14, Alpha=@n27) {constB=0.5}) {outputName=Sub Cells, sortPriority=10}) {name=Random Sub Cells, nodeColor=<rgba 0 0 0 1>}
  NamedRerouteDeclaration(None=FunctionInput() {bUsePreviewValueAsDefault=true, inputName=Random Seed, inputType=FunctionInput_Scalar, sortPriority=1000}) {name=Random Seed, nodeColor=<rgba 0 0 0 1>}
  NamedRerouteDeclaration(None=Rotator(
    Coordinate=MaterialFunctionCall!(ScreenUV=@n13, TextureSize, ScalePivotPoint) {materialFunction=/FortUILibrary/MaterialFunctions/Basics/MF_UI_ScreenSpaceUVScaling}.ScaleY
    Time=FunctionInput() {bUsePreviewValueAsDefault=true, inputName=Rotation, inputType=FunctionInput_Scalar, sortPriority=6}
   ) {speed=6.2832}) {name=Global UV, nodeColor=<rgba 0 0 0 1>}
  NamedRerouteDeclaration(None=FunctionInput(Preview=Time()) {bUsePreviewValueAsDefault=true, inputName=Time, inputType=FunctionInput_Scalar, sortPriority=2000}) {name=Time, nodeColor=<rgba 0 0 0 1>}
  MaterialExpressionFunctionInput_5 = FunctionInput() {
   bUsePreviewValueAsDefault=true
   inputName=Main Min/Max Opacity
   inputType=FunctionInput_Vector2
   sortPriority=105
   previewValue=<0 0.0625 0 1>
  }
  n28 = Convert(None=@MaterialExpressionFunctionInput_5) {
   nodeName=Make Vector2
   convertInputs [{
     type=Vector2
     expressionInput {
      expression=&self:MaterialExpressionFunctionInput_5
      inputName=None
      mask=0
      maskA=0
      maskB=0
      maskG=0
      maskR=0
      outputIndex=0
     }
    }]
   convertOutputs [{type=Scalar}, {type=Scalar}]
  }
  n29 = FunctionInput() {bUsePreviewValueAsDefault=true, inputName=Main Speed, inputType=FunctionInput_Scalar, sortPriority=150, previewValue=<0.1 0 0 1>}
  n30 = NamedRerouteUsage()
  n31 = Floor(None=@n30)
  n32 = NamedRerouteUsage()
  n33 = Subtract(
   A=Multiply(A=NamedRerouteUsage(), B=@n29) {constB=0.1}
   B=NamedRerouteDeclaration(None=MaterialFunctionCall(UV=@n31, Seed=@n32) {materialFunction=&MF_UI_Random}.Result) {name=Main Cells Random, nodeColor=<rgba 0 0 0 1>}
  )
  n34 = LinearInterpolate(
   A=Multiply(A=FunctionInput() {bUsePreviewValueAsDefault=true, inputName=Main Reduction, inputType=FunctionInput_Scalar, sortPriority=195}) {constB=-1}
   Alpha=ConstantBiasScale(None=Sine(None=@n33))
  )
  MaterialExpressionFunctionInput_14 = FunctionInput() {
   bUsePreviewValueAsDefault=true
   inputName=Main Min/Max Perspective
   inputType=FunctionInput_Vector2
   sortPriority=112
   previewValue=<0.1 0 0 1>
  }
  n35 = Convert(None=@MaterialExpressionFunctionInput_14) {
   nodeName=Make Vector2
   convertInputs [{
     type=Vector2
     expressionInput {
      expression=&self:MaterialExpressionFunctionInput_14
      inputName=None
      mask=0
      maskA=0
      maskB=0
      maskG=0
      maskR=0
      outputIndex=0
     }
    }]
   convertOutputs [{type=Scalar}, {type=Scalar}]
  }
  n36 = NamedRerouteUsage()
  n37 = FunctionInput() {
   bUsePreviewValueAsDefault=true
   inputName=Main Min/Max Scale
   inputType=FunctionInput_Vector2
   sortPriority=110
   previewValue=<0.925 0.75 0 1>
  }
  MaterialExpressionMax_4 = Max(A=@n37) {constB=0}
  n38 = Convert(None=@MaterialExpressionMax_4) {
   nodeName=Make Vector2
   convertInputs [{
     type=Vector2
     expressionInput {
      expression=&self:MaterialExpressionMax_4
      inputName=None
      mask=0
      maskA=0
      maskB=0
      maskG=0
      maskR=0
      outputIndex=0
     }
    }]
   convertOutputs [{type=Scalar}, {type=Scalar}]
  }
  n39 = MaterialFunctionCall(
   UVs=MaterialFunctionCall(
    UV=Frac(None=@n30)
    Perspective Rotation=NamedRerouteUsage()
    Perspective Amount=LinearInterpolate(A=@n35, B=@n35, Alpha=@n36) {constA=0.1, constB=0}
   ) {materialFunction=&MF_UI_Perspective}.Perspective UVs
   Size=LinearInterpolate(A=@n38, B=@n38, Alpha=NamedRerouteUsage()) {constA=0.75, constB=0}
   Corner Roundness=Constant() {r=0.2}
   Preserve Roundness=Constant()
  ) {materialFunction=&MF_SDF_Box_V4}
  n40 = MaterialFunctionCall(DistanceField=@n39.SDF) {materialFunction=&MF_UI_AntiAliasing}
  n41 = FunctionInput() {
   bUsePreviewValueAsDefault=true
   inputName=Main Outline Thickness
   inputType=FunctionInput_Scalar
   sortPriority=145
   previewValue=<0.025 0 0 1>
  }
  n42 = Constant() {r=-1}
  n43 = Multiply(
   A=NamedRerouteDeclaration(None=Saturate(None=@n34)) {name=Animated Main Cells, nodeColor=<rgba 0 0 0 1>}
   B=LinearInterpolate(
    A=FunctionOutput(None=@n40.Inner) {outputName=Main Mask, sortPriority=6}
    B=MaterialFunctionCall(DistanceField=@n39.SDF, Thickness=@n41, Alignment=@n42) {materialFunction=&MF_UI_ShapeOutline}.Outline
    Alpha=NamedRerouteUsage()
   )
  )
  NamedRerouteDeclaration(None=FunctionOutput(None=LinearInterpolate(A=@n28, B=@n28, Alpha=@n43) {constB=0.25}) {outputName=Main Cells, sortPriority=5}) {name=Random Main Cells, nodeColor=<rgba 0 0 0 1>}
 ]
}
